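# Single radial pip widget for the smart rifle's auto-aim reticle, showing lock-on charge through a dynamic material on one image.
# SetPipRotation angles the pip by its index within an evenly divided circle.
# UpdatePipChargePercentage starts a looping timer, and ProgressTimer ramps the material's Progress over the lock time, then flags full charge.
# OnTargetLocked plays a short material and opacity animation; ResetPip stops animations and timer and restores material parameters.
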WidgetBlueprint /SmartRifle/UI/Widgets/WBP_AutoAim_Reticle_Pip
animations [Anim_TargetLockedOn {
  binding_count=1
  length_seconds=0.15
  object=&self:Anim_TargetLockedOn
  sequence {bindings [Image_Pip {
     id=#1
     possessed_class=Image
     tracks [
      {
       type=MovieSceneWidgetMaterialTrack
       sections [{
         type=MovieSceneParameterSection
         end_frame=3
         start_frame=0
         channels [
          TargetLocked {type=MovieSceneScriptingFloatChannel, const [1,2]}
          BgOpacity {type=MovieSceneScriptingFloatChannel, const [1,2]}
          Anim {type=MovieSceneScriptingFloatChannel, const [1,2]}
          Progress {type=MovieSceneScriptingFloatChannel, const [0,2]}
         ]
        }]
      }
      {
       type=MovieSceneFloatTrack
       display_name=Render Opacity
       sections [{type=MovieSceneFloatSection, end_frame=3, start_frame=0, channels [None {type=MovieSceneScriptingFloatChannel, keys [[0,0],[3,1]]}]}]
      }
     ]
    }]}
 }]
graphs [
 EventGraph {links [], nodes []}
 ForceUpdateProgress {nodes [
   n1 = ForceUpdateProgress!() -> (then, Value:float=~) {is=title K2Node_FunctionEntry}
   > n2 = GetDynamicMaterial(self=Get Image_Pip() -> (Image_Pip:Image Object Reference))
   > SetScalarParameterValue(self=@n2, ParameterName=Progress, Value=@n1)
   > Set IsFullyCharged(IsFullyCharged:bool="float >= float"(A=@n1, B=1)) -> (Output_Get:bool)
  ]}
 OnTargetLocked {nodes [
   OnTargetLocked() {is=title K2Node_FunctionEntry}
   > PlayAnimation(InAnimation=Get Anim_TargetLockedOn() -> (Anim_TargetLockedOn:Widget Animation Object Reference))
  ]}
 ProgressTimer {nodes [
   ProgressTimer() {is=title K2Node_FunctionEntry}
   > n1 = Branch(Condition="float < float"(A=Get CachedTimeElapsed() -> (CachedTimeElapsed:float), B=Get Time to Lock Pip() -> (Time to Lock Pip:float)))
   > n2 = Set CachedTimeElapsed(CachedTimeElapsed:float=float + float(A=Get CachedTimeElapsed(), B=Get TimerUpdateRate() -> (TimerUpdateRate:double))) -> (Output_Get:float)
   > n3 = GetDynamicMaterial(self=Get Image_Pip())
   > SetScalarParameterValue(self=@n3, ParameterName=Progress, Value=MapRangeClamped(Value=@n2, InRangeA=0, InRangeB=Get Time to Lock Pip()))
   Clear and Invalidate Timer by Handle(execute=@n1.else, Handle=Get Progress_TimerHandle() -> (Progress_TimerHandle:Timer Handle Structure))
   > n4 = GetDynamicMaterial(self=Get Image_Pip())
   > SetScalarParameterValue(self=@n4, ParameterName=Progress, Value=1)
   > Set CachedTimeElapsed(CachedTimeElapsed=0)
   > Set IsFullyCharged(IsFullyCharged=true)
  ]}
 ResetPip {nodes [
   ResetPip() {is=title K2Node_FunctionEntry}
   > n1 = Sequence()
   > StopAllAnimations(self:User Widget Object Reference)
   > Set ChargedAnimPlayed(ChargedAnimPlayed:bool=false) -> (Output_Get:bool)
   > Clear and Invalidate Timer by Handle(Handle=Get Progress_TimerHandle())
   > ForceUpdateProgress!(execute, self, Value:float=0) -> (then) {is=title K2Node_CallFunction}
   n2 = GetDynamicMaterial(execute=@n1.then_1, self=Get Image_Pip())
   > SetScalarParameterValue(self=@n2, ParameterName=TargetLocked, Value=1)
   > SetScalarParameterValue(self=@n2, ParameterName=Anim, Value=1)
   > SetScalarParameterValue(self=@n2, ParameterName=BgOpacity, Value=1)
   > SetScalarParameterValue(self=@n2, ParameterName=Progress, Value=0)
  ]}
 SetPipRotation {nodes [
   n1 = SetPipRotation!() -> (then, PipIndex:int, MaxPipIndex:int) {is=title K2Node_FunctionEntry}
   > Set RadialSegmentsAngle(RadialSegmentsAngle:int=int / int(A=360, B=@n1.MaxPipIndex)) -> (Output_Get:int=0)
   n2 = Get PipIndex() -> (PipIndex:int=0)
   n3 = Get RadialSegmentsAngle() -> (RadialSegmentsAngle:int=0)
   > SetRenderTransformAngle!(execute, self:Widget Object Reference=Get Image_Pip(), Angle:float="To Float (Integer)"(InInt=int * int(A=@n3, B=@n2))) -> (then)
  ]}
 UpdatePipChargePercentage {nodes [
   n1 = UpdatePipChargePercentage!() -> (then, TimeToLockPip:float=~) {is=title K2Node_FunctionEntry}
   > Set Time to Lock Pip(Time to Lock Pip:float=@n1) -> (Output_Get:float)
   n2 = Get Progress_TimerHandle()
   > n3 = Branch(Condition=OR Boolean(A=Get IsFullyCharged() -> (IsFullyCharged:bool), B=Is Timer Active by Handle!(Handle:Timer Handle Structure=@n2) -> (ReturnValue:bool)))
   n4 = Set Timer by Event(execute=@n3.else, Delegate=Create Event(), Time=Get TimerUpdateRate(), bLooping=true, bMaxOncePerFrame=true)
   > Set Progress_TimerHandle(Progress_TimerHandle:Timer Handle Structure=@n4) -> (Output_Get:Timer Handle Structure)
  ]}
]
variables [
 CachedTimeElapsed {type=number}
 ChargedAnimPlayed {type=boolean}
 IsFullyCharged {type=boolean}
 Progress_TimerHandle {type=TimerHandle}
 Time to Lock Pip {type=number}
 TimerUpdateRate {type=number, default=0.1}
]
widgets {
 info {inheritedWidgetCount=0, parentClass=&UserWidget, rootWidgetClass=&Image, widgetCount=1}
 widgets [Image_Pip {
   widgetClassPath=&Image
   bIsVariable=true
   widget=Image_Pip
   properties {
    brush {
     bIsDynamicallyLoaded=false
     drawAs=Image
     imageType=FullColor
     mirroring=NoMirror
     resourceName=None
     resourceObject=/SmartRifle/UI/Materials/MI_UI_AutoAim_Reticle_LockOn
     tiling=NoTile
     imageSize=<45 45>
     margin=<ltrb 0 0 0 0>
     outlineSettings {
      bUseBrushTransparency=false
      roundingType=HalfHeightRadius
      width=0
      color {colorUseRule=UseColor_Specified, specifiedColor=<rgba 0 0 0 0>}
      cornerRadii=<0 0 0 0>
     }
     tintColor {colorUseRule=UseColor_Specified, specifiedColor=<rgba 1 1 1 1>}
     uVRegion {bIsValid=false, max=<0 0>, min=<0 0>}
    }
    renderTransform {angle=0, scale=<1 -1>, shear=<0 0>, translation=<0 0>}
   }
  }]
}
