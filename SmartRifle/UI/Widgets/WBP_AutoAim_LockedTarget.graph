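# Smart Rifle HUD widget that shows lock-on charge on a target as a radial ring of pips.
# On construct it finds the weapon instigator's auto-aim component, binds its events and creates pips from a curve-table max lock-on count.
# A change in lock-on permission toggles the ring and a looping timer for pip updates.
# Full charge stops the timer and fills locked pips; target loss hides and resets them.
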WidgetBlueprint /SmartRifle/UI/Widgets/WBP_AutoAim_LockedTarget
graphs [
 BindToAutoAimCompEvents {nodes [
   n1 = BindToAutoAimCompEvents!() -> (then, BindEvents:bool=~) {is=title K2Node_FunctionEntry}
   n2 = Get AutoAimRifle_Component() -> (AutoAimRifle_Component:Auto Aim Rifle Component Object Reference)
   > Is Valid(InputObject=@n2)
   > n3 = Branch(Condition=@n1)
   n4 = Create Event()
   > Bind Event to On Full Charge(self:Auto Aim Rifle Component Object Reference=@n2, Delegate=@n4) {is=title K2Node_AddDelegate}
   n5 = Create Event()
   > Bind Event to On Lockon Allowed(self:Auto Aim Rifle Component Object Reference=@n2, Delegate=@n5) {is=title K2Node_AddDelegate}
   n6 = Create Event()
   > Bind Event to On Lost Lock Target(self:Auto Aim Rifle Component Object Reference=@n2, Delegate=@n6) {is=title K2Node_AddDelegate}
   Unbind Event from On Full Charge(execute=@n3.else, self:Auto Aim Rifle Component Object Reference=@n2, Delegate=@n4) {is=title K2Node_RemoveDelegate}
   > Unbind Event from On Lockon Allowed(self:Auto Aim Rifle Component Object Reference=@n2, Delegate=@n5) {is=title K2Node_RemoveDelegate}
   > Unbind Event from On Lost Lock Target(self:Auto Aim Rifle Component Object Reference=@n2, Delegate=@n6) {is=title K2Node_RemoveDelegate}
  ]}
 ClearReticlePips {nodes [
   ClearReticlePips() {is=title K2Node_FunctionEntry}
   > Reset!(
    execute
    self:Dynamic Entry Box Object Reference=Get DynamicEntryBox_Pips() -> (DynamicEntryBox_Pips:Dynamic Entry Box Object Reference)
    bDeleteWidgets:bool=true
   ) -> (then)
   > Return Node()
  ]}
 EventGraph {nodes [
   n1 = UpdateLockOnAllowed!() -> (OutputDelegate, then, LockOnAllowed:bool=~) {is=title K2Node_CustomEvent}
   > n2 = Delta Gate Wildcard!(execute:Exec, Variable:"Boolean (by ref)"=Get LockOnAllowed() -> (LockOnAllowed:bool), New Value:bool=@n1.LockOnAllowed) -> (Success:Exec, Updated Variable:"Boolean (by ref)", Previous Variable:bool=~) {is=title K2Node_MacroInstance}
   > n3 = Branch(Condition=@n2)
   n4 = Get DynamicEntryBox_Pips()
   > SetVisibility(self=@n4, InVisibility=HitTestInvisible)
   > n5 = Set Timer by Event(Delegate=Create Event(), Time=0.1, bLooping=true, bMaxOncePerFrame=true)
   > Set Locked on Ammo Timer Handle(Locked on Ammo Timer Handle:Timer Handle Structure=@n5) -> (Output_Get:Timer Handle Structure)
   SetVisibility(execute=@n3.else, self=@n4, InVisibility=Hidden)
   > Clear and Invalidate Timer by Handle(Handle=Get Locked on Ammo Timer Handle() -> (Locked on Ammo Timer Handle:Timer Handle Structure))
   > ResetReticlePips()
   Event Construct()
   > n6 = Cast To FortPawn!(execute, Object=GetOwningPlayerPawn!(self:User Widget Object Reference) -> (ReturnValue:Pawn Object Reference)) -> (then, CastFailed, AsFort Pawn:Fort Pawn Object Reference) {is=title K2Node_DynamicCast}
   n7 = GetCurrentWeapon!(self:Fort Pawn Object Reference=@n6) -> (ReturnValue:Fort Weapon Object Reference)
   > n8 = Set AutoAimRifle_Component(AutoAimRifle_Component:Auto Aim Rifle Component Object Reference=GetComponentByClass!(
     self:Actor Object Reference=Get Instigator!(self:Actor Object Reference=@n7) -> (Instigator:Pawn Object Reference)
     ComponentClass:Actor Component Class Reference=/SmartRifle/Gameplay/AutoAim_Rifle_Component.AutoAim_Rifle_Component_C
    ) -> (ReturnValue:Auto Aim Rifle Component Object Reference)) -> (Output_Get:Auto Aim Rifle Component Object Reference)
   > Set BulletsPerBurst(BulletsPerBurst:int=GetBurstCountStat!(self:Auto Aim Rifle Component Object Reference=@n8) -> (BurstCount:int=0)) -> (Output_Get:int=0)
   > BindToAutoAimCompEvents!(execute, self, BindEvents:bool=true) -> (then) {is=title K2Node_CallFunction}
   n9 = Get Max Lock Ons() -> (Max Lock Ons:Scalable Float Structure)
   > GenerateReticlePips!(
    execute
    self
    ReticlePipsToDisplay:int=Truncate(A=GetValueAtLevel!(Input:"Scalable Float Structure (by ref)"=@n9, Level:float=1) -> (ReturnValue:float))
   ) -> (then)
   Event Destruct()
   > Parent: Destruct()
   > BindToAutoAimCompEvents(BindEvents=false)
  ]}
 GenerateReticlePips {nodes [
   n1 = GenerateReticlePips!() -> (then, ReticlePipsToDisplay:int) {is=title K2Node_FunctionEntry}
   > ClearReticlePips() {is=title K2Node_CallFunction}
   > n2 = For Loop(LastIndex=int - int(A=@n1))
   ResetReticlePips(execute=@n2.Completed)
   n3 = Create Entry!(execute=@n2.LoopBody, self:Dynamic Entry Box Object Reference=Get DynamicEntryBox_Pips()) -> (then, ReturnValue:User Widget Object Reference)
   > n4 = Cast To WBP_AutoAim_Reticle_Pip!(execute, Object=@n3) -> (then, CastFailed, AsWBP Auto Aim Reticle Pip:WBP Auto Aim Reticle Pip Object Reference) {is=title K2Node_DynamicCast}
   > SetPipRotation!(execute, self:WBP Auto Aim Reticle Pip Object Reference=@n4, PipIndex:int=@n2, MaxPipIndex:int=@n1) -> (then)
  ]}
 GetPipEntries {nodes [
   GetPipEntries() {is=title K2Node_FunctionEntry}
   > Return Node!(
    execute
    ReticlePipEntries:Array of User Widget Object References=GetAllEntries!(self:Dynamic Entry Box Base Object Reference=Get DynamicEntryBox_Pips()) -> (ReturnValue:Array of User Widget Object References)
   )
  ]}
 OnFullCharge {nodes [
   OnFullCharge() {is=title K2Node_FunctionEntry}
   > Clear and Invalidate Timer by Handle(Handle=Get Locked on Ammo Timer Handle())
   n1 = Get AutoAimRifle_Component()
   > n2 = For Loop(LastIndex=int - int(A=GetCurrentLockOnCount!(self:Auto Aim Weapon Pawn Component Object Reference=@n1) -> (ReturnValue:int=0)))
   > n3 = Cast To WBP_AutoAim_Reticle_Pip(Object="Get (a copy)"(
     Array:Array of User Widget Object References=GetPipEntries!() -> (ReticlePipEntries:Array of User Widget Object References) {is=title K2Node_CallFunction}
     Dimension 1=@n2
    ) -> (Output:User Widget Object Reference))
   > n4 = Branch(Condition=Get IsFullyCharged!(self:WBP Auto Aim Reticle Pip Object Reference=@n3) -> (IsFullyCharged:bool))
   ForceUpdateProgress!(execute=@n4.else, self:WBP Auto Aim Reticle Pip Object Reference=@n3, Value:float=1) -> (then)
  ]}
 OnGainedLockTarget {nodes [
   OnGainedLockTarget() {is=title K2Node_FunctionEntry}
   > n1 = For Each Loop(Array:Array of User Widget Object References=GetAllEntries(self=Get DynamicEntryBox_Pips())) -> (Array Element:User Widget Object Reference)
   > n2 = Cast To WBP_AutoAim_Reticle_Pip(Object=@n1)
   > OnTargetLocked(self:WBP Auto Aim Reticle Pip Object Reference=@n2)
  ]}
 OnLostTarget {nodes [
   OnLostTarget() {is=title K2Node_FunctionEntry}
   > SetVisibility(self=Get DynamicEntryBox_Pips(), InVisibility=Hidden)
   > Clear and Invalidate Timer by Handle(Handle=Get Locked on Ammo Timer Handle())
   > ResetReticlePips()
   > Set LockOnAllowed(LockOnAllowed:bool=false) -> (Output_Get:bool)
  ]}
 ResetReticlePips {nodes [
   ResetReticlePips() {is=title K2Node_FunctionEntry}
   > n1 = For Each Loop(Array=GetAllEntries(self=Get DynamicEntryBox_Pips()))
   Return Node!(execute=@n1.Completed)
   > n2 = Cast To WBP_AutoAim_Reticle_Pip(Object=@n1)
   > ResetPip(self:WBP Auto Aim Reticle Pip Object Reference=@n2)
  ]}
 UpdateLockedOnPipState {nodes [
   UpdateLockedOnPipState() {is=title K2Node_FunctionEntry}
   n1 = Get AutoAimRifle_Component()
   n2 = GetCurrentLockOnCount(self=@n1) -> (ReturnValue=0)
   > n3 = Cast To WBP_AutoAim_Reticle_Pip(Object="Get (a copy)"(Array=GetPipEntries(), Dimension 1=@n2))
   > UpdatePipChargePercentage!(
    execute
    self:WBP Auto Aim Reticle Pip Object Reference=@n3
    TimeToLockPip:float=GetLockOnTime!(self:Auto Aim Rifle Component Object Reference=@n1, LockOnCount:int=@n2) -> (LockOnTime:double)
   ) -> (then)
  ]}
]
variables [
 AutoAimRifle_Component {type=AutoAim_Rifle_Component_C, default=null}
 BulletsPerBurst {type=integer}
 Locked on Ammo Timer Handle {type=TimerHandle}
 LockOnAllowed {type=boolean}
 Max Lock Ons {
  type=ScalableFloat
  default {registryType=None, value=1, curve {curveTable=/SmartRifle/Gameplay/SmartRifleGameData, rowName=Default.Smartgun.MaxLockOns}}
 }
]
widgets {
 info {inheritedWidgetCount=1, parentClass=/Script/FortniteUI.FortWeaponAttachmentPointWidget, rootWidgetClass=&Overlay, widgetCount=4}
 widgets [
  Overlay_Root {widgetClassPath=&Overlay, widget=Overlay_Root}
  SizeBox_0 {
   widgetClassPath=&SizeBox
   parent=Overlay_Root
   widget=SizeBox_0
   properties {bOverride_HeightOverride=true, bOverride_WidthOverride=true, heightOverride=90, widthOverride=90}
   slot {class=OverlaySlot, properties {horizontalAlignment=HAlign_Center, verticalAlignment=VAlign_Center}}
  }
  DynamicEntryBox_Pips {
   widgetClassPath=&DynamicEntryBox
   bIsVariable=true
   parent=SizeBox_0
   widget=DynamicEntryBox_Pips
   properties {
    entryBoxType=Radial
    entryWidgetClass=&/WBP_AutoAim_Reticle_Pip.WBP_AutoAim_Reticle_Pip_C
    numDesignerPreviewEntries=4
    radialBoxSettings {angleBetweenItems=0, bClockwiseOrder=true, bDistributeItemsEvenly=true, marginSize=0, sectorCentralAngle=360, startingAngle=90}
    renderTransform {angle=45, scale=<1 1>, shear=<0 0>, translation=<0 0>}
   }
   slot {class=SizeBoxSlot}
  }
  TutorialHighlight {widgetClassPath=&TutorialHighlight}
 ]
}
