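# Translucent UI material that draws the Smart Rifle auto-aim ammo bar as segmented charges, each split into pips.
# Tiles horizontal UVs into spaced charges and pips with Frac, mirroring the vertical axis for pip shapes.
# Steps reversed U against MeterProgress and DeltaProgress to mask empty and loaded charges.
# Blends two box SDFs, then derives anti-aliased outline, fill and shadow color and opacity, with edge-pixel clipping.
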Material /SmartRifle/UI/Materials/M_UI_AutoAim_AmmoBar
graph {
 properties {bAutomaticallySetUsageInEditor=false, blendMode=BLEND_Translucent, materialDomain=MD_UI, usedShadingModels=Default Lit}
 nodes [
  n1 = LinearInterpolate(
   A=MaterialFunctionCall(UVs=NamedRerouteUsage(), Size=Constant2Vector() {g=0.5, r=0.35}) {materialFunction=&MF_UI_Box}.SDF
   B=MaterialFunctionCall(UVs=NamedRerouteUsage(), Size=Constant2Vector() {g=1.2, r=0.46}) {materialFunction=&MF_UI_Box}.SDF
   Alpha=NamedRerouteUsage()
  )
  n2 = LinearInterpolate(
   A=VectorParameter() {parameterName=OutlineColor, defaultValue=<rgba 1 0.0513 0 1>}.RGB
   B=VectorParameter() {parameterName=FillColor, defaultValue=<rgba 1 0.5457 0.1559 1>}.RGB
   Alpha=MaterialFunctionCall(DistanceField=@n1, Threshold=Constant() {r=-0.225}) {materialFunction=&MF_UI_AntiAliasing}.Inner
  )
  n3 = NamedRerouteUsage()
  n4 = MaterialFunctionCall(DistanceField=@n1) {materialFunction=&MF_UI_AntiAliasing}
  LinearInterpolate(B=LinearInterpolate(A=LinearInterpolate(B=@n2, Alpha=@n3), Alpha=NamedRerouteUsage()), Alpha=@n4.Inner)
  n5 = MaterialFunctionCall(DistanceField=@n1, Threshold=Constant() {r=0.25}) {materialFunction=&MF_UI_AntiAliasing}
  n6 = LinearInterpolate(A=@n4.Inner, B=@n5.Inner) {desc=ShadowOpacity}
  n7 = Multiply(A=@n6) {constB=0.4, desc=OffPipOpacity}
  n8 = NamedRerouteUsage()
  MaterialFunctionCall!(
   UVs=NamedRerouteUsage()
   Alpha=LinearInterpolate(A=LinearInterpolate(A=@n7, B=@n5.Inner, Alpha=@n8), B=Multiply(A=@n6) {constB=0.8}, Alpha=NamedRerouteUsage())
  ) {materialFunction=/FortUILibrary/MaterialFunctions/Basics/MF_UI_ClipEdgePixel}
  n9 = OneMinus(None=ComponentMask(None=MaterialFunctionCall() {materialFunction=&GetUserInterfaceUV}.Normalized UV) {r=true})
  n10 = Divide(A=Constant() {r=1}, B=ScalarParameter() {defaultValue=8, parameterName=Number of charges, sortPriority=2})
  n11 = Step(Y=@n9, X=Multiply(A=@n10, B=ScalarParameter() {defaultValue=5, parameterName=MeterProgress, sortPriority=1}) {desc=Empty Charges}) {constX=6}
  NamedRerouteDeclaration(None=@n11) {name=EmptyCharges, nodeColor=<rgba 0 0 0 1>}
  n12 = Step(Y=@n9, X=Multiply(A=@n10, B=ScalarParameter() {defaultValue=7, parameterName=DeltaProgress, sortPriority=2}) {desc=Loaded charges}) {constX=6}
  NamedRerouteDeclaration(None=Saturate(None=Subtract(A=@n12, B=@n11))) {name=LoadedCharges, nodeColor=<rgba 0 0 0 1>}
  n13 = MaterialFunctionCall() {materialFunction=&GetUserInterfaceUV}
  n14 = Frac(None=Multiply(A=ScalarParameter() {defaultValue=8, parameterName=Number of charges, sortPriority=2}, B=ComponentMask(None=@n13.Normalized UV) {r=true}))
  n15 = ScalarParameter() {defaultValue=0.65, parameterName=SpaceInBetween, sortPriority=4}
  n16 = Multiply(
   A=NamedRerouteDeclaration(None=MaterialFunctionCall(UVs=@n14, Texture Scale=@n15) {materialFunction=&ScaleUVsByCenter}.UVs) {name=UVChargesWithSpacing, nodeColor=<rgba 0 0 0 1>}
   B=ScalarParameter() {defaultValue=3, parameterName=Pips per charge, sortPriority=3}
  )
  n17 = ComponentMask(None=@n13.Normalized UV) {g=true}
  n18 = Constant() {r=1}
  n19 = Add(
   A=MaterialFunctionCall!(UV=Multiply(A=@n17, B=@n18), Mirror Y, Swizzle UV) {materialFunction=/FortUILibrary/MaterialFunctions/Basics/MF_UI_Symmetry_2x1_New}.Symmetry
   B=Constant2Vector() {r=-0.35}
  )
  n20 = Constant() {r=1}
  NamedRerouteDeclaration(None=AppendVector(A=Frac(None=@n16), B=Multiply(A=@n19, B=@n20))) {name=PipsUV, nodeColor=<rgba 0 0 0 1>}
 ]
}
