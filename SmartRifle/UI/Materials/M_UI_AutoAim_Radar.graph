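# M_UI_AutoAim_Radar is a translucent UI material that procedurally draws a radar graphic for the Smart Rifle's auto-aim.
# Scaled and rotated box SDFs, combined by intersection, subtraction and union, form an anti-aliased inner shape, outline and box mask.
# A reflected gradient supplies a glow mask, and 2x2 symmetry mirrors small trapezoid details.
# Named reroutes carry these masks, with lerp, multiply and max steps blending mask values.
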Material /SmartRifle/UI/Materials/M_UI_AutoAim_Radar
graph {
 properties {bAutomaticallySetUsageInEditor=false, blendMode=BLEND_Translucent, materialDomain=MD_UI, usedShadingModels=Default Lit}
 nodes [
  Constant4Vector() {constant=<rgba 1 1 1 1>}
  n1 = MaterialFunctionCall!(DistanceFieldA=NamedRerouteUsage(), DistanceFieldB=NamedRerouteUsage(), BlendAmount) {materialFunction=/FortUILibrary/MaterialFunctions/Shapes/Modifiers/MF_UI_SDF_Intersection}
  MaterialFunctionCall(DistanceFieldA=@n1.Exact, DistanceFieldB=NamedRerouteUsage()) {materialFunction=&MF_UI_SDF_Intersection}
  n2 = NamedRerouteUsage()
  n3 = NamedRerouteUsage()
  n4 = Multiply(A=Multiply(A=NamedRerouteUsage(), B=@n3)) {constB=0.2}
  Multiply(A=Max(A=LinearInterpolate(A=@n2, B=@n4, Alpha=@n3), B=NamedRerouteUsage()), B=Constant() {r=0.9})
  NamedRerouteDeclaration(None=MaterialFunctionCall(
    UVs=MaterialFunctionCall() {materialFunction=&GetUserInterfaceUV}.Normalized UV
    Texture Scale=Constant2Vector() {g=1.7, r=1.9}
   ) {materialFunction=&ScaleUVsByCenter}.UVs) {name=UV, nodeColor=<rgba 0 0 0 1>}
  NamedRerouteDeclaration(None=MaterialFunctionCall(UVs=NamedRerouteUsage()) {materialFunction=&MF_UI_Box}.SDF) {name=Box01_SDF, nodeColor=<rgba 0 0 0 1>}
  n5 = NamedRerouteUsage()
  n6 = Constant() {r=0.125}
  NamedRerouteDeclaration(None=MaterialFunctionCall(
    UVs=MaterialFunctionCall(UVs=@n5, "Rotation Angle (0-1)"=@n6) {materialFunction=&CustomRotator}.Rotated Values
    Size=Constant() {r=0.625}
   ) {materialFunction=&MF_UI_Box}.SDF) {name=Box02_SDF, nodeColor=<rgba 0 0 0 1>}
  NamedRerouteDeclaration(None=MaterialFunctionCall(DistanceField=@n1.Exact) {materialFunction=&MF_UI_AntiAliasing}.Inner) {name=InnerShape, nodeColor=<rgba 0 0 0 1>}
  n7 = NamedRerouteUsage()
  NamedRerouteDeclaration(None=MaterialFunctionCall(
    DistanceField=MaterialFunctionCall!(DistanceFieldA=@n1.Exact, DistanceFieldB=@n7, BlendAmount) {materialFunction=/FortUILibrary/MaterialFunctions/Shapes/Modifiers/MF_UI_SDF_Subtraction}.Exact
    Threshold=Constant() {r=0.003}
   ) {materialFunction=&MF_UI_AntiAliasing}.Inner) {name=Outline, nodeColor=<rgba 0 0 0 1>}
  n8 = NamedRerouteUsage()
  NamedRerouteDeclaration(None=OneMinus(None=MaterialFunctionCall!(UVs=@n8, Offset, RotationCenter, RotationAngle) {materialFunction=/FortUILibrary/MaterialFunctions/Gradients/MF_UI_ReflectedGradient}.Result)) {name=GlowMask, nodeColor=<rgba 0 0 0 1>}
  n9 = NamedRerouteUsage()
  n10 = Constant2Vector() {g=1, r=0.35}
  n11 = Constant2Vector() {g=0.326, r=1}
  NamedRerouteDeclaration(None=MaterialFunctionCall!(
    DistanceFieldA=MaterialFunctionCall(UVs=@n9, Size=@n10) {materialFunction=&MF_UI_Box}.SDF
    DistanceFieldB=MaterialFunctionCall(UVs=@n9, Size=@n11) {materialFunction=&MF_UI_Box}.SDF
    BlendAmount
   ) {materialFunction=/FortUILibrary/MaterialFunctions/Shapes/Modifiers/MF_UI_SDF_Union}.Exact) {name=BoxMask, nodeColor=<rgba 0 0 0 1>}
  n12 = MaterialFunctionCall(UVs=NamedRerouteUsage(), Texture Scale=Constant2Vector() {g=3.1, r=0.66}) {materialFunction=&ScaleUVsByCenter}
  n13 = Constant() {r=0.125}
  NamedRerouteDeclaration(None=MaterialFunctionCall(
    UVs=MaterialFunctionCall(UVs=@n12.UVs, "Rotation Angle (0-1)"=@n13) {materialFunction=&CustomRotator}.Rotated Values
    Size=Constant() {r=0.61}
   ) {materialFunction=&MF_UI_Box}.SDF) {name=Box03_SDF, nodeColor=<rgba 0 0 0 1>}
  n14 = NamedRerouteUsage()
  n15 = Constant() {r=0.05}
  n16 = Add(
   A=MaterialFunctionCall!(
    UV=MaterialFunctionCall(UVs=@n14, Texture Scale=@n15) {materialFunction=&ScaleUVsByCenter}.UVs
    Swizzle UV
   ) {materialFunction=/FortUILibrary/MaterialFunctions/Basics/MF_UI_Symmetry_2x2_New}.Flipped Symmetry
   B=Constant2Vector() {g=-9.42, r=-6.8}
  )
  n17 = Constant() {r=0.2}
  NamedRerouteDeclaration(None=MaterialFunctionCall(DistanceField=MaterialFunctionCall!(UVs=@n16, TopWidth, BottomWidth, Height=@n17, OutlineThickness, OutlineAlignment, GlowMin, GlowMax) {materialFunction=/FortUILibrary/MaterialFunctions/Shapes/MF_UI_IsoscelesTrapezoid}.SDF) {materialFunction=&MF_UI_AntiAliasing}.Inner) {name=TrapezoidDetails, nodeColor=<rgba 0 0 0 1>}
 ]
}
