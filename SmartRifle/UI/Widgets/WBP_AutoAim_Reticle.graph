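# Reticle extension widget for the auto-aim smart rifle: crosshair, lock-on radius and burst-segmented ammo bar.
# Aiming down sights plays an animation revealing radius and bar while rotating the crosshair; hip-fire spread drives a crosshair material parameter.
# Weapon, HUD and auto-aim component events update bar material parameters, with looping timers for reload refill and lock-on count.
# Lock gain in ADS and lock loss broadcast reticle state through the global event router.
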WidgetBlueprint /SmartRifle/UI/Widgets/WBP_AutoAim_Reticle
animations [
 Anim_Ads {
  binding_count=3
  length_seconds=0.2501
  object=&self:Anim_Ads
  sequence {bindings [
    Overlay_Ammo {
     id=#1
     possessed_class=Overlay
     tracks [{
       type=MovieSceneFloatTrack
       display_name=Render Opacity
       sections [{type=MovieSceneFloatSection, end_frame=5, start_frame=0, channels [None {type=MovieSceneScriptingFloatChannel, keys [[0,0],[5,1]]}]}]
      }]
    }
    Image_Crosshairs {
     id=#2
     possessed_class=Image
     tracks [{
       type=MovieSceneWidgetMaterialTrack
       sections [{
         type=MovieSceneParameterSection
         end_frame=5
         start_frame=0
         channels [
          CrosshairOpacity {type=MovieSceneScriptingFloatChannel, keys [[0,1],[2,0],[2.5957,0],[5,1]]}
          ReticleRotation {type=MovieSceneScriptingFloatChannel, keys [[0,0.125],[2,0.125],[2.4163,0],[5,0]]}
         ]
        }]
      }]
    }
    Image_Radius {
     id=#3
     possessed_class=Image
     tracks [{
       type=MovieSceneFloatTrack
       display_name=Render Opacity
       sections [{type=MovieSceneFloatSection, end_frame=5, start_frame=0, channels [None {type=MovieSceneScriptingFloatChannel, keys [[0,0],[5,1]]}]}]
      }]
    }
   ]}
 }
 Anim_LockedOn {binding_count=0, length_seconds=0.25, object=&self:Anim_LockedOn, sequence {bindings []}}
]
graphs [
 BindToAutoAimCompEvents {nodes [
   BindToAutoAimCompEvents!() -> (then, BindEvents:bool=~) {is=title K2Node_FunctionEntry}
   n1 = Get AutoAimRifle_Weapon() -> (AutoAimRifle_Weapon:B Auto Aim Rifle Object Reference)
   > Is Valid(InputObject=@n1)
   n2 = GetComponentByClass!(
    self:Actor Object Reference=Get Instigator!(self:Actor Object Reference=@n1) -> (Instigator:Pawn Object Reference)
    ComponentClass:Actor Component Class Reference=/SmartRifle/Gameplay/AutoAim_Rifle_Component.AutoAim_Rifle_Component_C
   ) -> (ReturnValue:Auto Aim Rifle Component Object Reference)
   > Is Valid(InputObject=@n2)
   > Set AutoAimRifle_Component(AutoAimRifle_Component:Auto Aim Rifle Component Object Reference=@n2) -> (Output_Get:Auto Aim Rifle Component Object Reference)
   > n3 = Branch(Condition=Get BindEvents() -> (BindEvents:bool))
   n4 = Create Event()
   n5 = Get AutoAimRifle_Component() -> (AutoAimRifle_Component:Auto Aim Rifle Component Object Reference)
   > Bind Event to On Full Charge(self:Auto Aim Rifle Component Object Reference=@n5, Delegate=@n4) {is=title K2Node_AddDelegate}
   n6 = Create Event()
   > Bind Event to On Lockon Allowed(self:Auto Aim Rifle Component Object Reference=@n5, Delegate=@n6) {is=title K2Node_AddDelegate}
   n7 = Create Event()
   > Bind Event to On Gained Lock Target(self:Auto Aim Rifle Component Object Reference=@n5, Delegate=@n7) {is=title K2Node_AddDelegate}
   n8 = Create Event()
   > Bind Event to On Lost Lock Target(self:Auto Aim Rifle Component Object Reference=@n5, Delegate=@n8) {is=title K2Node_AddDelegate}
   > Set BulletsPerBurst(BulletsPerBurst:int=GetBurstCountStat!(self:Auto Aim Rifle Component Object Reference=@n5) -> (BurstCount:int=0)) -> (Output_Get:int=0)
   Unbind Event from On Full Charge(execute=@n3.else, self:Auto Aim Rifle Component Object Reference=@n5, Delegate=@n4) {is=title K2Node_RemoveDelegate}
   > Unbind Event from On Lockon Allowed(self:Auto Aim Rifle Component Object Reference=@n5, Delegate=@n6) {is=title K2Node_RemoveDelegate}
   > Unbind Event from On Gained Lock Target(self:Auto Aim Rifle Component Object Reference=@n5, Delegate=@n7) {is=title K2Node_RemoveDelegate}
   > Unbind Event from On Lost Lock Target(self:Auto Aim Rifle Component Object Reference=@n5, Delegate=@n8) {is=title K2Node_RemoveDelegate}
  ]}
 CacheDynamicMaterialRefs {nodes [
   CacheDynamicMaterialRefs() {is=title K2Node_FunctionEntry}
   > n1 = GetDynamicMaterial(self=Get Image_AmmoBar() -> (Image_AmmoBar:Image Object Reference))
   > Set DynamicMat_AmmoBar(DynamicMat_AmmoBar:Material Instance Dynamic Object Reference=@n1) -> (Output_Get:Material Instance Dynamic Object Reference)
   > n2 = GetDynamicMaterial(self=Get Image_Crosshairs() -> (Image_Crosshairs:Image Object Reference))
   > Set DynamicMat_Crosshair(DynamicMat_Crosshair:Material Instance Dynamic Object Reference=@n2) -> (Output_Get:Material Instance Dynamic Object Reference)
  ]}
 CreateBindings {nodes [
   CreateBindings() {is=title K2Node_FunctionEntry}
   > n1 = Cast To B_AutoAim_Rifle!(
    execute
    Object=GetEquippedWeapon!(self:Fort Weapon Reticle Extension Widget Base Object Reference) -> (ReturnValue:Fort Weapon Object Reference)
   ) -> (then, CastFailed, AsB Auto Aim Rifle:B Auto Aim Rifle Object Reference) {is=title K2Node_DynamicCast}
   > n2 = Set AutoAimRifle_Weapon(AutoAimRifle_Weapon:B Auto Aim Rifle Object Reference=@n1) -> (Output_Get:B Auto Aim Rifle Object Reference)
   > Bind Event to On Local Ammo Changed(self:Fort Weapon Object Reference=@n2, Delegate=Create Event()) {is=title K2Node_AddDelegate}
   > Bind Event to On Spread Radius Changed(
    self:Fort HUDContext Object Reference=Get FortHUDContext!() -> (ReturnValue:Fort HUDContext Object Reference) {is=title K2Node_GetBlueprintContextEx}
    Delegate=Create Event()
   ) {is=title K2Node_AddDelegate}
   > ResetAmmoBar()
   > ShowLockedAttachmentPoint!(execute, self, ShouldShow:bool=false) -> (then)
  ]}
 EventGraph {nodes [
   Event Destruct()
   > Unbind Event from On Local Ammo Changed(self:Fort Weapon Object Reference=Get AutoAimRifle_Weapon(), Delegate=Create Event()) {is=title K2Node_RemoveDelegate}
   > Unbind Event from On Spread Radius Changed(self:Fort HUDContext Object Reference=Get FortHUDContext(), Delegate=Create Event()) {is=title K2Node_RemoveDelegate}
   > BindToAutoAimCompEvents!(execute, self, BindEvents:bool=false) -> (then) {is=title K2Node_CallFunction}
   > ResetADSState()
   > Set AutoAimRifle_Component()
   > Set AutoAimRifle_Weapon()
   Event Construct()
   > CreateBindings() {is=title K2Node_CallFunction}
   > BindToAutoAimCompEvents(BindEvents=true)
   n1 = Event OnWeaponReloadStarted!() -> (OutputDelegate, then, ReloadTime:float, ReloadType:EFortWeaponReloadType Enum=ReloadWholeClip) {is=title K2Node_Event}
   > n2 = Set Reload Time(Reload Time:float=@n1.ReloadTime) -> (Output_Get:float)
   > Set IsReloading(IsReloading:bool=true) -> (Output_Get:bool)
   n3 = Get TotalBurstsPerClip() -> (TotalBurstsPerClip:int=0)
   > n4 = Set Timer by Event(Delegate=Create Event(), Time=float / float(A=@n2, B="To Float (Integer)"(InInt=@n3)), bLooping=true)
   > Set ReloadAmmo_TimerHandle(ReloadAmmo_TimerHandle:Timer Handle Structure=@n4) -> (Output_Get:Timer Handle Structure)
   n5 = Event OnTargetingChanged!() -> (OutputDelegate, then, bIsADS:bool) {is=title K2Node_Event}
   > UpdateAutoAimRifleComponentBindings!(execute, self) -> (then, AutoAimRifleComponent:Auto Aim Rifle Component Object Reference)
   > n6 = Set InADS(InADS:bool=@n5.bIsADS) -> (Output_Get:bool)
   > n7 = Branch(Condition=@n6)
   n8 = Get Anim_Ads() -> (Anim_Ads:Widget Animation Object Reference)
   > PlayAnimationForward(InAnimation=@n8)
   PlayAnimationReverse(execute=@n7.else, InAnimation=@n8)
   Event OnInitialized()
   > CacheDynamicMaterialRefs() {is=title K2Node_CallFunction}
   n9 = Event PreConstruct()
   > Branch(Condition=@n9.IsDesignTime)
   > SetRenderOpacity(
    self=[Get Image_Radius() -> (Image_Radius:Image Object Reference), Get Overlay_Ammo() -> (Overlay_Ammo:Overlay Object Reference)]
    InOpacity=1
   )
   Event OnWeaponReloadCancelled() {is=title K2Node_Event}
   > Clear and Invalidate Timer by Handle(Handle=Get ReloadAmmo_TimerHandle() -> (ReloadAmmo_TimerHandle:Timer Handle Structure))
   > Set IsReloading(IsReloading=false)
   > ResetAmmoBar()
   > Set ReloadedCharges(ReloadedCharges:double=0) -> (Output_Get:double)
  ]}
 GetAutoAimRifleComponent {nodes [
   GetAutoAimRifleComponent() {is=title K2Node_FunctionEntry}
   n1 = Get AutoAimRifle_Component()
   > n2 = Is Valid(InputObject=@n1)
   n3 = UpdateAutoAimRifleComponentBindings(execute=@n2.Is Not Valid)
   > Return Node!(execute, AutoAimRifleComponent:Auto Aim Rifle Component Object Reference=@n3)
   Return Node(execute=@n2.Is Valid, AutoAimRifleComponent=@n1)
  ]}
 GetAvailableBurstCount {nodes [
   GetAvailableBurstCount() {is=title K2Node_FunctionEntry}
   n1 = Get AutoAimRifle_Weapon()
   > Return Node!(
    execute
    ReturnValue:int=int / int(
     A=GetLocalMagazineAmmoCount!(self:Fort Weapon Object Reference=@n1) -> (ReturnValue:int=0)
     B=Get BulletsPerBurst() -> (BulletsPerBurst:int=0)
    )
   )
  ]}
 OnFullCharge {nodes [
   OnFullCharge() {is=title K2Node_FunctionEntry}
   > Clear and Invalidate Timer by Handle(Handle=Get LockedOnAmmo_TimerHandle() -> (LockedOnAmmo_TimerHandle:Timer Handle Structure))
   > UpdateLockedOnAmmo()
  ]}
 OnGainedLockTarget {nodes [
   OnGainedLockTarget() {is=title K2Node_FunctionEntry}
   > Branch(Condition=Get InADS() -> (InADS:bool))
   > ShowLockedAttachmentPoint(ShouldShow=true)
  ]}
 OnLocalAmmoChanged {nodes [
   OnLocalAmmoChanged!() -> (then, LocalCount:int, LocalRemaining:int) {is=title K2Node_FunctionEntry}
   > n1 = Branch(Condition=Get IsReloading() -> (IsReloading:bool))
   n2 = GetAutoAimRifleComponent!(execute=@n1.else, self) -> (then, AutoAimRifleComponent:Auto Aim Rifle Component Object Reference) {is=title K2Node_CallFunction}
   > Is Valid(InputObject=@n2)
   > n3 = Branch(Condition="integer > integer"(A=GetCurrentLockOnCount!(self:Auto Aim Weapon Pawn Component Object Reference=@n2) -> (ReturnValue:int=0)))
   n4 = "To Float (Integer)"(InInt=GetAvailableBurstCount!() -> (ReturnValue:int=0) {is=title K2Node_CallFunction})
   n5 = Get DynamicMat_AmmoBar() -> (DynamicMat_AmmoBar:Material Instance Dynamic Object Reference)
   > SetScalarParameterValue(self=@n5, ParameterName=DeltaProgress, Value=@n4)
   SetScalarParameterValue(execute=@n3.else, self=@n5, ParameterName=MeterProgress, Value=@n4)
   > SetScalarParameterValue(self=@n5, ParameterName=DeltaProgress, Value=0)
  ]}
 OnLostLockTarget {nodes [
   OnLostLockTarget() {is=title K2Node_FunctionEntry}
   > ResetAmmoBar()
   > ShowLockedAttachmentPoint(ShouldShow=false)
   > Set CanLockOn(CanLockOn:bool=false) -> (Output_Get:bool)
  ]}
 OnReloadStarted {nodes [
   OnReloadStarted() {is=title K2Node_FunctionEntry}
   > n1 = Set ReloadedCharges(ReloadedCharges=float + float(A=Get ReloadedCharges() -> (ReloadedCharges:double), B=1))
   n2 = Get TotalBurstsPerClip() -> (TotalBurstsPerClip=0)
   n3 = int / int(
    A=GetRemainingAmmo!(self:Fort Weapon Object Reference=Get AutoAimRifle_Weapon()) -> (ReturnValue:int=0)
    B=Get BulletsPerBurst() -> (BulletsPerBurst=0)
   )
   n4 = SelectFloat(A="To Float (Integer)"(InInt=@n3), B="To Float (Integer)"(InInt=@n2), bPickA="integer < integer"(A=@n3, B=@n2))
   > n5 = Branch(Condition="Equal (Float)"(A=@n1, B=@n4))
   n6 = Get DynamicMat_AmmoBar()
   > SetScalarParameterValue(self=@n6, ParameterName=MeterProgress, Value=@n4)
   > Clear and Invalidate Timer by Handle(Handle=Get ReloadAmmo_TimerHandle())
   > Set IsReloading(IsReloading=false)
   > Set ReloadedCharges(ReloadedCharges=0)
   SetScalarParameterValue(execute=@n5.else, self=@n6, ParameterName=MeterProgress, Value=@n1)
   > SetScalarParameterValue(self=@n6, ParameterName=DeltaProgress, Value=0)
  ]}
 OnSpreadRadiusChanged {nodes [
   n1 = OnSpreadRadiusChanged!() -> (then, SpreadRadius:float=~) {is=title K2Node_FunctionEntry}
   > n2 = Branch(Condition=Get InADS())
   SetScalarParameterValue(
    execute=@n2.else
    self=Get DynamicMat_Crosshair() -> (DynamicMat_Crosshair:Material Instance Dynamic Object Reference)
    ParameterName=SpreadOffset
    Value=@n1
   )
  ]}
 ResetADSState {nodes [
   ResetADSState() {is=title K2Node_FunctionEntry}
   > Set InADS(InADS=false)
   > SetRenderOpacity(self=[Get Overlay_Ammo(), Get Image_Radius()], InOpacity=0)
   n1 = Get DynamicMat_Crosshair()
   > SetScalarParameterValue(self=@n1, ParameterName=CrosshairOpacity, Value=1)
   > SetScalarParameterValue(self=@n1, ParameterName=ReticleRotation, Value=0.125)
   PlayAnimation(InAnimation=Get Anim_Ads(), PlayMode=Reverse)
  ]}
 ResetAmmoBar {nodes [
   ResetAmmoBar() {is=title K2Node_FunctionEntry}
   n1 = Get AutoAimRifle_Weapon()
   > n2 = Set TotalBurstsPerClip(TotalBurstsPerClip:int=int / int(A=GetBulletsPerClip!(self:Fort Weapon Object Reference=@n1) -> (ReturnValue:int=0), B=Get BulletsPerBurst() -> (BulletsPerBurst=0))) -> (Output_Get:int=0)
   n3 = Get DynamicMat_AmmoBar()
   > SetScalarParameterValue(self=@n3, ParameterName=" Number of charges", Value="To Float (Integer)"(InInt=@n2))
   > SetScalarParameterValue(self=@n3, ParameterName=Pips per charge, Value="To Float (Integer)"(InInt=Get BulletsPerBurst() -> (BulletsPerBurst=0)))
   > SetScalarParameterValue(self=@n3, ParameterName=MeterProgress, Value="To Float (Integer)"(InInt=GetAvailableBurstCount() -> (ReturnValue=0)))
   > SetScalarParameterValue(self=@n3, ParameterName=DeltaProgress, Value=0)
  ]}
 ShowLockedAttachmentPoint {nodes [
   n1 = ShowLockedAttachmentPoint!() -> (then, ShouldShow:bool=~) {is=title K2Node_FunctionEntry}
   > n2 = Branch(Condition=@n1)
   n3 = GetGlobalEventRouter()
   > Broadcast Event!(
    execute
    self:Gameplay Event Router Component Object Reference=@n3
    EventData:Fort Event Lock On Reticle State Changed Structure=Make Fort Event Lock on Reticle State Changed!(bEnabledReticle:bool=true) -> (FortEvent_LockOnReticleStateChanged:Fort Event Lock On Reticle State Changed Structure) {is=title K2Node_MakeStruct}
    EventContext:Object Reference
   ) -> (then, ReturnValue:bool)
   Broadcast Event(execute=@n2.else, self=@n3, EventData=Make Fort Event Lock on Reticle State Changed(bEnabledReticle=False))
  ]}
 UpdateAutoAimRifleComponentBindings {nodes [
   UpdateAutoAimRifleComponentBindings() {is=title K2Node_FunctionEntry}
   > n1 = Is Valid(InputObject=Get AutoAimRifle_Component())
   n2 = Get AutoAimRifle_Weapon()
   n3 = Is Valid(exec=@n1.Is Not Valid, InputObject=@n2)
   Return Node!(execute=@n3.Is Not Valid, AutoAimRifleComponent:Auto Aim Rifle Component Object Reference)
   n4 = GetComponentByClass(self=Get Instigator(self=@n2), ComponentClass=&AutoAim_Rifle_Component_C)
   > n5 = Is Valid(InputObject=@n4)
   Return Node(execute=@n5.Is Not Valid)
   > n6 = Set AutoAimRifle_Component(AutoAimRifle_Component=@n4)
   > BindToAutoAimCompEvents(BindEvents=true)
   > Return Node(AutoAimRifleComponent=@n6)
  ]}
 UpdateLockedOnAllowed {nodes [
   n1 = UpdateLockedOnAllowed!() -> (then, LockOnAllowed:bool=~) {is=title K2Node_FunctionEntry}
   > n2 = Delta Gate Wildcard!(execute:Exec, Variable:"Boolean (by ref)"=Get CanLockOn() -> (CanLockOn:bool), New Value:bool=@n1) -> (Success:Exec, Updated Variable:"Boolean (by ref)", Previous Variable:bool=~) {is=title K2Node_MacroInstance}
   > n3 = Branch(Condition=@n2)
   > n4 = Set Timer by Event(Delegate=Create Event(), Time=0.1, bLooping=true)
   > Set LockedOnAmmo_TimerHandle(LockedOnAmmo_TimerHandle:Timer Handle Structure=@n4) -> (Output_Get:Timer Handle Structure)
   Clear and Invalidate Timer by Handle(execute=@n3.else, Handle=Get LockedOnAmmo_TimerHandle())
  ]}
 UpdateLockedOnAmmo {nodes [
   UpdateLockedOnAmmo() {is=title K2Node_FunctionEntry}
   > n1 = GetAutoAimRifleComponent()
   > Is Valid(InputObject=@n1)
   > Branch()
   n2 = GetCurrentLockOnCount(self=@n1) -> (ReturnValue=0)
   n3 = GetAvailableBurstCount() -> (ReturnValue=0)
   n4 = Get DynamicMat_AmmoBar()
   > SetScalarParameterValue(self=@n4, ParameterName=MeterProgress, Value="To Float (Integer)"(InInt=int - int(A=@n3, B=@n2)))
   > SetScalarParameterValue(self=@n4, ParameterName=DeltaProgress, Value="To Float (Integer)"(InInt=GetAvailableBurstCount() -> (ReturnValue=0)))
  ]}
]
variables [
 AutoAimRifle_Component {type=AutoAim_Rifle_Component_C, default=null}
 AutoAimRifle_Weapon {type=B_AutoAim_Rifle_C, default=null}
 BulletsPerBurst {type=integer}
 CanLockOn {type=boolean}
 DynamicMat_AmmoBar {type=MaterialInstanceDynamic}
 DynamicMat_Crosshair {type=MaterialInstanceDynamic}
 InADS {type=boolean}
 IsReloading {type=boolean}
 LockedOnAmmo_TimerHandle {type=TimerHandle}
 Max Lock Ons {
  type=ScalableFloat
  default {registryType=None, value=1, curve {curveTable=/SmartRifle/Gameplay/SmartRifleGameData, rowName=Default.Smartgun.MaxLockOns}}
 }
 Reload Time {type=number}
 ReloadAmmo_TimerHandle {type=TimerHandle}
 ReloadedCharges {type=number}
 TotalBurstsPerClip {type=integer}
]
widgets {
 info {
  inheritedWidgetCount=1
  parentClass=/Script/FortniteUI.FortWeaponReticleExtensionWidgetBase
  rootWidgetClass=&Overlay
  widgetCount=7
 }
 widgets [
  Overlay_Root {widgetClassPath=&Overlay, widget=Overlay_Root}
  Overlay_Reticle {
   widgetClassPath=&Overlay
   parent=Overlay_Root
   widget=Overlay_Reticle
   slot {class=OverlaySlot, properties {horizontalAlignment=HAlign_Fill, verticalAlignment=VAlign_Fill}}
  }
  Image_Crosshairs {
   widgetClassPath=&Image
   bIsVariable=true
   parent=Overlay_Reticle
   widget=Image_Crosshairs
   properties {brush {
     bIsDynamicallyLoaded=false
     drawAs=Image
     imageType=FullColor
     mirroring=NoMirror
     resourceName=None
     resourceObject=/SmartRifle/UI/Materials/MI_UI_AutoAim_Reticle_Cross
     tiling=NoTile
     imageSize=<64 64>
     margin=<ltrb 0 0 0 0>
     outlineSettings {
      bUseBrushTransparency=false
      roundingType=HalfHeightRadius
      width=0
      color {colorUseRule=UseColor_Specified, specifiedColor=<rgba 0 0 0 0>}
      cornerRadii=<0 0 0 0>
     }
     tintColor {colorUseRule=UseColor_Specified, specifiedColor=<rgba 1 1 1 1>}
     uVRegion {bIsValid=false, max=<0 0>, min=<0 0>}
    }}
   slot {class=OverlaySlot, properties {horizontalAlignment=HAlign_Center, verticalAlignment=VAlign_Center}}
  }
  Image_Radius {
   widgetClassPath=&Image
   bIsVariable=true
   parent=Overlay_Reticle
   widget=Image_Radius
   properties {
    renderOpacity=0
    brush {
     bIsDynamicallyLoaded=false
     drawAs=Image
     imageType=FullColor
     mirroring=NoMirror
     resourceName=None
     resourceObject=/SmartRifle/UI/Materials/M_UI_AutoAim_Radar
     tiling=NoTile
     imageSize=<400 400>
     margin=<ltrb 0 0 0 0>
     outlineSettings {
      bUseBrushTransparency=false
      roundingType=HalfHeightRadius
      width=0
      color {colorUseRule=UseColor_Specified, specifiedColor=<rgba 0 0 0 0>}
      cornerRadii=<0 0 0 0>
     }
     tintColor {colorUseRule=UseColor_Specified, specifiedColor=<rgba 1 1 1 1>}
     uVRegion {bIsValid=false, max=<0 0>, min=<0 0>}
    }
   }
   slot {class=OverlaySlot, properties {horizontalAlignment=HAlign_Center, verticalAlignment=VAlign_Center}}
  }
  Overlay_Ammo {
   widgetClassPath=&Overlay
   bIsVariable=true
   parent=Overlay_Reticle
   widget=Overlay_Ammo
   properties {renderOpacity=0, visibility=Visible}
   slot {class=OverlaySlot, properties {horizontalAlignment=HAlign_Center, verticalAlignment=VAlign_Bottom, padding=<ltrb 0 0 0 40>}}
  }
  Image_AmmoBar {
   widgetClassPath=&Image
   bIsVariable=true
   parent=Overlay_Ammo
   widget=Image_AmmoBar
   properties {brush {
     bIsDynamicallyLoaded=false
     drawAs=Image
     imageType=FullColor
     mirroring=NoMirror
     resourceName=None
     resourceObject=/SmartRifle/UI/Materials/M_UI_AutoAim_AmmoBar
     tiling=NoTile
     imageSize=<270 26>
     margin=<ltrb 0 0 0 0>
     outlineSettings {
      bUseBrushTransparency=false
      roundingType=HalfHeightRadius
      width=0
      color {colorUseRule=UseColor_Specified, specifiedColor=<rgba 0 0 0 0>}
      cornerRadii=<0 0 0 0>
     }
     tintColor {colorUseRule=UseColor_Specified, specifiedColor=<rgba 1 1 1 1>}
     uVRegion {bIsValid=false, max=<0 0>, min=<0 0>}
    }}
   slot {class=OverlaySlot, properties {horizontalAlignment=HAlign_Center, verticalAlignment=VAlign_Center}}
  }
  TutorialHighlight {widgetClassPath=&TutorialHighlight}
 ]
}
